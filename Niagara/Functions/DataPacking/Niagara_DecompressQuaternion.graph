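# Niagara library material function that decodes a quaternion packed into one scalar and outputs it as a float4.
# A Custom node reinterprets the bits as smallest-three encoding: a 2-bit largest-component index and three 10-bit values.
# It rescales those to ±1/sqrt(2), rebuilds the fourth from unit length and swizzles by index.
# The extraction uses logical `&&` with the mask instead of bitwise `&`, so components lose their 10-bit values.
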MaterialFunction /Niagara/Functions/DataPacking/Niagara_DecompressQuaternion
graph {
 properties {bExposeToLibrary=true, libraryCategoriesText [Niagara]}
 nodes [
  MaterialExpressionFunctionInput_0 = FunctionInput() {inputName=CompressedQuaternion, inputType=FunctionInput_Scalar}
  n1 = Custom!(EncodedQuatFloat=@MaterialExpressionFunctionInput_0) {
   code="uint maxIndex = 0;\r\nfloat4 Quat = float4(1,1,1,1);\r\n\r\n\r\nuint EncodedQuat = asuint(EncodedQuatFloat);\r\nint CompressionBits = 10;\r\nint CompressionMask = (1 << CompressionBits)-1; //1023, precision mask\r\n\r\nmaxIndex = EncodedQuat >> (30);\r\n\r\nQuat.x = float( (EncodedQuat >> (CompressionBits * 2) ) && CompressionMask );\r\nQuat.y = float( (EncodedQuat >> (CompressionBits * 1) ) && CompressionMask );\r\nQuat.z = float( (EncodedQuat >> (CompressionBits * 0) ) && CompressionMask );\r\nQuat.xyz /= CompressionMask;\r\n\r\nconst float scale = 0.707106781;     // 1/sqrt(2)\r\n\r\nQuat.xyz = ( ( Quat.xyz ) * ( scale + scale ) ) - scale;\r\n\r\nQuat.w = sqrt(max(0.0f, 1.0f + dot(Quat.xyz, -Quat.xyz)));\r\n\r\nif(maxIndex == 0)\r\n    {\r\n    Quat = Quat.wxyz;\r\n    }\r\nelse if (maxIndex == 1)\r\n    {\r\n    Quat = Quat.xwyz;\r\n    }\r\nelse if (maxIndex == 2)\r\n    {\r\n    Quat = Quat.xywz;\r\n    }\r\nelse if (maxIndex == 3)\r\n    {\r\n    Quat = Quat;\r\n    }\r\n\r\nreturn Quat;"
   outputType=CMOT_Float4
   inputs [{inputName=EncodedQuatFloat, input {expression=&self:MaterialExpressionFunctionInput_0}}]
  }
  FunctionOutput(None=@n1)
 ]
}
