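# Material function for Niagara, exposed in the library, that rotates a mesh particle's vertices around the particle position using a quaternion.
# It takes world position relative to particle position and rotates it with MultiplyVectorWithQuaternion, with a static switch input to invert the quaternion.
# It then subtracts the unrotated relative vector, outputting an offset from the original position, the form World Position Offset takes.
# World and particle positions are overridable inputs that default to the current values.
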MaterialFunction /Niagara/Functions/DataPacking/Niagara_RotateMeshParticleWithQuaternion
graph {
 properties {bExposeToLibrary=true, libraryCategoriesText [Niagara]}
 nodes [
  n1 = FunctionInput(Preview=WorldPosition().XYZ) {bUsePreviewValueAsDefault=true, inputName=World Position, sortPriority=2}
  n2 = FunctionInput(Preview=ParticlePositionWS()) {bUsePreviewValueAsDefault=true, inputName=Particle Position, sortPriority=3}
  n3 = Subtract(A=@n1, B=@n2)
  n4 = FunctionInput() {inputName=Quaternion, inputType=FunctionInput_Vector4}
  n5 = FunctionInput(Preview=StaticBool()) {bUsePreviewValueAsDefault=true, inputName=Invert Quaternion, inputType=FunctionInput_StaticBool, sortPriority=1}
  FunctionOutput(None=Add(
    A=MaterialFunctionCall!(Vector=@n3, Quat=@n4, Invert Quaternion=@n5) {materialFunction=/Engine/Functions/Engine_MaterialFunctions03/Math/MultiplyVectorWithQuaternion}.Result
    B=Subtract(A=@n2, B=@n1)
   ))
 ]
}
